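# Sound cue for the impact sound of a file cabinet interior prop in the demolition impact set.
# A random node picks one of three equally weighted file cabinet impact waves, and a modulator varies its pitch slightly on each play.
# It uses small melee weapon attenuation and the impacts sound class, with concurrency limited by a shared object-impact set.
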SoundCue /Figment_S07/Sounds/Fort_Impact_Sounds/Demolition/Interior_Props/Demo_FileCabinet_Impact_Cue
root=Modulator_10
nodes [
 {id=Modulator_10, children [Random_25], properties {pitchMax=1.07, pitchMin=0.92}}
 {id=Random_25, children [WavePlayer_164, WavePlayer_165, WavePlayer_166], properties {weights [1,1,1]}}
 {id=WavePlayer_164, properties {soundWaveAssetPtr=/Figment_S07/Sounds/Fort_Impact_Sounds/Demolition/FileCabinet_Impact01}}
 {id=WavePlayer_165, properties {soundWaveAssetPtr=/Figment_S07/Sounds/Fort_Impact_Sounds/Demolition/FileCabinet_Impact02}}
 {id=WavePlayer_166, properties {soundWaveAssetPtr=/Figment_S07/Sounds/Fort_Impact_Sounds/Demolition/FileCabinet_Impact03}}
]
properties {
 attenuationSettings=&Fort_Weapon_Melee_Sml_Att
 maxDistance=2800
 soundClassObject=&Fort_Impacts
 volumeMultiplier=0.8
 concurrencyOverrides {resolutionRule=PreventNew}
 concurrencySet [&CONC_Object_Impacts_FourMax]
}
